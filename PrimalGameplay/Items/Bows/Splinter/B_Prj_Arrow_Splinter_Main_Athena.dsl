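; Main arrow projectile for the splinter bow, which spawns secondary arrows around itself a short delay after BeginPlay.
; A timer fires DoSplinter, which spawns either a set count at random sign-flipped offsets within min/max bounds or one per entry in a fixed offset array.
; Offsets are world-space, unrotated by the arrow's facing.
; Secondaries copy the main arrow's rotation, current speed, gravity scale, instigator and on-touch effect container.
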(event Collision|EventActorBeginOverlap
  (params (OtherActor :type Actor%20Object%20Reference))
  (|CallParentFunction OtherActor) ; This node is disabled and will not be called.\nDrag off pins to build functionality.
)

(event EventTick
  (params (DeltaSeconds :type Float%20%28single-precision%29))
  (|CallParentFunction DeltaSeconds) ; This node is disabled and will not be called.\nDrag off pins to build functionality.
)

(event EventBeginPlay
  (|CallParentFunction)
  (Utilities|Time|SetTimerbyEvent (delegate Custom|DoSplinter) (Variables|Default|GetSplinterDelay)))

(event Custom|DoSplinter
  (if (Variables|Default|GetbUseRandomOffsetForSecondaryArrows)
    (for _ (range 1 (+ (Variables|Default|GetNumSecondaryProjectiles) 1))
      (CallFunction|SpawnSecondaryProjectileRandomPosition))
    (else
      (for _array_element (Variables|Default|GetSplinterArrowOffsets)
        (CallFunction|SpawnSecondaryProjectile (+ _array_element (Transformation|GetActorLocation)))))))

(fn GetRandomSign
  (return (Math|Float|SelectFloat 1.0 -1.0 (Math|Random|RandomBoolwithWeight 0.5))))

(fn GetRandomSpawnPosition
  (params (MinOffset :type Vector) (MaxOffset :type Vector))
  (bind _output_get (LocalVariables|SetOffset (Math|Vector|MakeVector (* (Math|Random|RandomFloatinRange (.x MinOffset) (.x MaxOffset)) (CallFunction|GetRandomSign)) (* (Math|Random|RandomFloatinRange (.y MinOffset) (.y MaxOffset)) (CallFunction|GetRandomSign)) (* (Math|Random|RandomFloatinRange (.z MinOffset) (.z MaxOffset)) (CallFunction|GetRandomSign)))))
  (return (+ _output_get (Transformation|GetActorLocation))))

(fn SpawnSecondaryProjectile
  (params (Location :type Vector))
  (Projectile|SpawnProjectile :ProjectileClass (Variables|Default|GetSecondaryProjectileClass) :RequestedBy (Variables|Actor|GetInstigator) :SpawnLocation Location :SpawnRotation (Transformation|GetActorRotation) :SpawnDirection (Transformation|GetActorRotation) :EffectContainerSpecToApplyOnHit (Variables|Delivery|GetEffectContainerSpecToApplyOnTouch) :bUseDefaultPhysics false :InitialSpeed (Math|Vector|VectorLength (Variables|/Script/Engine.MovementComponent|GetVelocity (Variables|Movement|GetProjectileMovementComponent))) :GravityScale (Variables|/Script/Engine.ProjectileMovementComponent|GetProjectileGravityScale (Variables|Movement|GetProjectileMovementComponent)) :ChargePercent 1.0))

(fn SpawnSecondaryProjectileRandomPosition
  (bind _spawnposition (CallFunction|GetRandomSpawnPosition (Variables|Default|GetSplinterMinOffset) (Variables|Default|GetSplinterMaxOffset)))
  (CallFunction|SpawnSecondaryProjectile _spawnposition))

(fn ConstructionScript
  (|CallParentFunction))
